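# Sound cue for collecting iron ingredient loot, routed to the loot collect sound class with prop collect attenuation.
# A mixer layers three waves at once: a realistic iron collect, a surreal iron collect, and a bag foley sound mixed slightly lower.
# A concurrency override caps simultaneous plays at two, stopping the oldest.
# Overall volume is reduced.
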SoundCue /Game/Sounds/Foley_Loot/Ingredients/Metal/Iron_Collect_Cue
root=Mixer_26
nodes [
 {id=Mixer_26, children [WavePlayer_54, WavePlayer_55, WavePlayer_6], properties {inputVolume [1,1,0.7]}}
 {id=WavePlayer_54, properties {soundWaveAssetPtr=&/Iron_Collect_Real}}
 {id=WavePlayer_55, properties {soundWaveAssetPtr=&/Iron_Collect_Surreal}}
 {id=WavePlayer_6, properties {soundWaveAssetPtr=/Game/Sounds/Foley_Loot/Ingredients/Bag/loot_setup_bagonly_01}}
]
properties {
 attenuationSettings=/Game/Sounds/Attenuation/Fort_Prop_Collect_Att
 bOverrideConcurrency=true
 maxDistance=2097152
 soundClassObject=&Fort_Loot_Collect
 volumeMultiplier=0.376
 concurrencyOverrides {maxCount=2, resolutionRule=StopOldest}
}
