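# Sound cue for collecting metal ingredients, a loot foley sound in the loot collect sound class.
# A Random node picks one of four equally weighted metal impact waves and feeds a Modulator at default settings for slight pitch and volume variation.
# Playback uses weapon pickup attenuation, reduced volume and an inline concurrency override that allows two instances, stopping the oldest.
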SoundCue /Game/Sounds/Foley_Loot/Ingredients/Metal/Metal_Collect_Cue
root=Modulator_2
nodes [
 {id=Modulator_2, children [Random_0]}
 {
  id=Random_0
  children [WavePlayer_0,WavePlayer_3,WavePlayer_4,WavePlayer_5]
  properties {bShouldExcludeFromBranchCulling=true, weights [1,1,1,1]}
 }
 {id=WavePlayer_0, properties {soundWaveAssetPtr=&/Metal_Collect_Impact_01}}
 {id=WavePlayer_3, properties {soundWaveAssetPtr=&/Metal_Collect_Impact_02}}
 {id=WavePlayer_4, properties {soundWaveAssetPtr=&/Metal_Collect_Impact_03}}
 {id=WavePlayer_5, properties {soundWaveAssetPtr=&/Metal_Collect_Impact_04}}
]
properties {
 attenuationSettings=&Fort_Weapon_Pickup_Att
 bOverrideConcurrency=true
 maxDistance=3500
 priority=0.3
 soundClassObject=&Fort_Loot_Collect
 volumeMultiplier=0.478
 concurrencyOverrides {maxCount=2, resolutionRule=StopOldest}
 concurrencySet [/Game/Sounds/Concurrency_Rules/OneMaximum_Owner_StopLowestPriority]
}
