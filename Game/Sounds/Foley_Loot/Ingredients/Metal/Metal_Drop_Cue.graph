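# Sound cue for metal ingredient drop impacts, part of the loot foley set.
# A fixed short delay feeds a modulator that varies pitch upward, then an equally weighted random pick among five impact waves.
# It uses weapon-drop attenuation and the loot-collect sound class at reduced volume.
# A concurrency override caps simultaneous plays at a low count, stopping the oldest.
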SoundCue /Game/Sounds/Foley_Loot/Ingredients/Metal/Metal_Drop_Cue
root=Delay_0
nodes [
 {id=Delay_0, children [Modulator_0], properties {delayMax=0.4, delayMin=0.4}}
 {id=Modulator_0, children [Random_0], properties {pitchMax=1.2, pitchMin=1}}
 {
  id=Random_0
  children [WavePlayer_0,WavePlayer_1,WavePlayer_2,WavePlayer_3,WavePlayer_4]
  properties {bShouldExcludeFromBranchCulling=true, weights [1,1,1,1,1]}
 }
 {id=WavePlayer_0, properties {soundWaveAssetPtr=&/Metal_Drop_Impact_01}}
 {id=WavePlayer_1, properties {soundWaveAssetPtr=&/Metal_Drop_Impact_02}}
 {id=WavePlayer_2, properties {soundWaveAssetPtr=&/Metal_Drop_Impact_03}}
 {id=WavePlayer_3, properties {soundWaveAssetPtr=&/Metal_Drop_Impact_04}}
 {id=WavePlayer_4, properties {soundWaveAssetPtr=&/Metal_Drop_Impact_05}}
]
properties {
 attenuationSettings=&Fort_Weapon_Drop_Att
 bOverrideConcurrency=true
 maxDistance=3000
 soundClassObject=&Fort_Loot_Collect
 volumeMultiplier=0.4
 concurrencyOverrides {maxCount=2, resolutionRule=StopOldest}
}
